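# Master material for skinned characters that layers an optional galaxy skin effect over the shared character shading, using material attributes.
# The base PreFX output feeds the Galaxy function, with SkinFX mask channels driving alpha and warp, then passes into PostFX.
# A static switch sources the galaxy's alternate fresnel from swirly noise, either panned in UV space or projected through a preskinned-position function.
# Both pan axes read GspeedX, leaving GspeedY unused.
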Material /Game/Characters/Player/Male/Medium/Bodies/M_Med_Soldier_04/Skins/Galaxy/Materials/M_FN_Celestial_Master
graph {
 properties {
  bUseMaterialAttributes=true
  bUsedWithMeshDeformer=true
  bUsedWithMorphTargets=true
  bUsedWithSkeletalMesh=true
  pixelDepthOffsetMode=PDOM_Legacy
  usedShadingModels=Default Lit
 }
 nodes [
  n1 = MaterialFunctionCall() {materialFunction=&BaseFN_CharacterFunction_PreFX}
  n2 = TextureSample(Tex=TextureObjectParameter() {group=Skin FX, parameterName=SkinFX_Mask}) {samplerType=SAMPLERTYPE_Masks}
  n3 = StaticComponentMaskParameter(None=@n2.RGBA) {defaultR=true, group=Skin FX, parameterName=Galaxy Warp Mask Channel}
  n4 = TextureObject() {samplerType=SAMPLERTYPE_LinearColor, texture=/Game/Weapons/WeaponSkins/Textures/T_Swirly_Noise_highContrast1}
  n5 = VectorParameter() {group=Galaxy, parameterName=GalaxyEF, channelNames {a="", b=GspeedY, g=GspeedX, r=EdgeFresTile}, defaultValue=<rgba 1 0 0.1 1>}
  n6 = AppendVector(A=@n5.GspeedX, B=@n5.GspeedX)
  n7 = TextureCoordinate() {uTiling=2, vTiling=2}
  n8 = TextureSample(UVs=Panner(Coordinate=Multiply(A=@n7, B=@n5.EdgeFresTile), Speed=@n6)) {
   samplerSource=SSM_Wrap_WorldGroupSettings
   samplerType=SAMPLERTYPE_LinearColor
   texture=&T_Swirly_Noise_highContrast1
  }
  n9 = StaticSwitchParameter(
   True=ComponentMask(None=MaterialFunctionCall!(TextureObject=@n4, TextureSize=@n5.EdgeFresTile, WorldPosition, Export Float 4, In=@n6, World Space Normal, ProjectionTransitionContrast) {materialFunction=/Game/Athena/SafeZone/Preskinned_WP_Texture_ScalePan}.XYZ Texture) {r=true}
   False=@n8.R
  ) {group=Galaxy, parameterName=UseLocalEdgeFresnel}
  n10 = StaticComponentMaskParameter(None=@n2.RGBA) {defaultR=true, group=Skin FX, parameterName=Galaxy Channel}
  MaterialFunctionCall(
   AO=@n1.AO
   Distance Blend=@n1.AO
   Fresnel=@n1.AO
   Gmap=@n1.AO
   Material=StaticSwitchParameter(
    True=MaterialFunctionCall!(In=@n1.Result, Warp Mask=@n3, GalaxyAlternateFresnel=@n9, Alpha=@n10) {materialFunction=/Game/Weapons/WeaponSkins/MaterialFunction/Galaxy}.Result
    False=@n1.Result
   ) {group=Skin FX, parameterName=Use Galaxy}
  ) {materialFunction=&BaseFN_CharacterFunction_PostFX}
 ]
}
